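# Sound cue for a wet cave ambience layer, built from dripping-water loops and scattered drip one-shots.
# A mixer blends a quiet constant basement-drip loop, a random pick of two other drip loops, and a louder one-shot branch.
# That branch repeatedly waits a random few seconds, then plays one of nine jungle-temple drips through a modulator.
# Output uses Athena ambient attenuation and the indoor ambient sound class.
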SoundCue /AmbientAudioContent/SoundCues/Biomes/Ambience_Cave_Wet_Layer_Loops_Cue
root=Mixer_0
nodes [
 {id=Delay_0, children [Modulator_0], properties {delayMax=6, delayMin=3}}
 {id=Looping_0, children [Delay_0]}
 {id=Mixer_0, children [WavePlayer_2,Random_0,Looping_0], properties {inputVolume [0.1,0.4,1.2]}}
 {id=Modulator_0, children [Random_1]}
 {id=Random_0, children [WavePlayer_0,WavePlayer_1], properties {weights [1,1]}}
 {
  id=Random_1
  children [WavePlayer_3, WavePlayer_6, WavePlayer_7, WavePlayer_8, WavePlayer_9, WavePlayer_10, WavePlayer_11, WavePlayer_12, WavePlayer_13]
  properties {weights [1,1,1,1,1,1,1,1,1]}
 }
 {id=WavePlayer_0, properties {bLooping=true, soundWaveAssetPtr=/AmbientAudioContent/Waves/POI/Ambience_ACabin_Interior_BasementDrips_04}}
 {id=WavePlayer_1, properties {bLooping=true, soundWaveAssetPtr=/AmbientAudioContent/Waves/POI/Ambience_ACabin_Interior_BasementDrips_03}}
 {id=WavePlayer_10, properties {soundWaveAssetPtr=/AmbientAudioContent/Waves/OneShots/Spot_Canary_Jungle_Temple_Int_SS_Drips_02_04}}
 {id=WavePlayer_11, properties {soundWaveAssetPtr=/AmbientAudioContent/Waves/OneShots/Spot_Canary_Jungle_Temple_Int_SS_Drips_02_03}}
 {id=WavePlayer_12, properties {soundWaveAssetPtr=/AmbientAudioContent/Waves/OneShots/Spot_Canary_Jungle_Temple_Int_SS_Drips_02_02}}
 {id=WavePlayer_13, properties {soundWaveAssetPtr=/AmbientAudioContent/Waves/OneShots/Spot_Canary_Jungle_Temple_Int_SS_Drips_02_01}}
 {id=WavePlayer_2, properties {bLooping=true, soundWaveAssetPtr=/AmbientAudioContent/Waves/POI/Ambience_ACabin_Interior_BasementDrips_02}}
 {id=WavePlayer_3, properties {soundWaveAssetPtr=/AmbientAudioContent/Waves/OneShots/Spot_Canary_Jungle_Temple_Int_SS_Drips_02_09}}
 {id=WavePlayer_6, properties {soundWaveAssetPtr=/AmbientAudioContent/Waves/OneShots/Spot_Canary_Jungle_Temple_Int_SS_Drips_02_08}}
 {id=WavePlayer_7, properties {soundWaveAssetPtr=/AmbientAudioContent/Waves/OneShots/Spot_Canary_Jungle_Temple_Int_SS_Drips_02_07}}
 {id=WavePlayer_8, properties {soundWaveAssetPtr=/AmbientAudioContent/Waves/OneShots/Spot_Canary_Jungle_Temple_Int_SS_Drips_02_06}}
 {id=WavePlayer_9, properties {soundWaveAssetPtr=/AmbientAudioContent/Waves/OneShots/Spot_Canary_Jungle_Temple_Int_SS_Drips_02_05}}
]
properties {
 attenuationSettings=&Fort_Ambient_Athena_Att
 maxDistance=2097152
 soundClassObject=&Fort_Ambient_Indoors_Athena
 volumeMultiplier=0.702
 preEffectBusSends [{}]
}
