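# Looping outdoor ambience cue of loud Japanese cicadas, using the Fort ambient Athena attenuation and sound class.
# A quality-level node selects quad wave loops for its first two levels and stereo versions for the last two.
# Each branch randomly picks one of three looping volume envelopes over a shared pitch envelope, which feeds a random choice of two loops.
# A root 72-second volume envelope layers a slow dip over the shorter cycles, varying the bed over time.
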SoundCue /AmbientAudioContent/SoundCues/Biomes/Ambience_Japanese_Cicadas_Loud_Only_Loop_Cue
root=Enveloper_3
nodes [
 {
  id=Enveloper_0
  children [Random_1]
  properties {
   bLoop=true
   bLoopIndefinitely=true
   loopEnd=30
   pitchCurve {editorCurveData {keys [{arriveTangent=-0.0013, interpMode=RCIM_Cubic, time=0}, <at 15.9468 0.96 RCIM_Cubic>, {interpMode=RCIM_Cubic, leaveTangent=0.0014, time=30}]}}
   volumeCurve {editorCurveData {keys [<at 0 1>]}}
  }
 }
 {
  id=Enveloper_1
  children [Random_0]
  properties {
   bLoop=true
   bLoopIndefinitely=true
   loopEnd=30
   pitchCurve {editorCurveData {keys [{arriveTangent=-0.0013, interpMode=RCIM_Cubic, time=0}, <at 15.9468 0.96 RCIM_Cubic>, {interpMode=RCIM_Cubic, leaveTangent=0.0014, time=30}]}}
   volumeCurve {editorCurveData {keys [<at 0 1>]}}
  }
 }
 {
  id=Enveloper_2
  children [Enveloper_1]
  properties {
   bLoop=true
   bLoopIndefinitely=true
   loopEnd=40
   pitchCurve {editorCurveData {keys [{time=0}]}}
   volumeCurve {editorCurveData {keys [
      {arriveTangent=-0.01, interpMode=RCIM_Cubic, time=0, value=1}
      <at 20 0.5 RCIM_Cubic>
      {interpMode=RCIM_Cubic, leaveTangent=0.01, time=39.9, value=1}
     ]}}
  }
 }
 {
  id=Enveloper_3
  children [QualityLevel_0]
  properties {
   bLoop=true
   bLoopIndefinitely=true
   loopEnd=72
   pitchCurve {editorCurveData {keys [{time=0}]}}
   volumeCurve {editorCurveData {keys [
      {arriveTangent=-0.01, interpMode=RCIM_Cubic, time=0, value=1}
      <at 36 0.5 RCIM_Cubic>
      {interpMode=RCIM_Cubic, leaveTangent=0.01, time=71.9, value=1}
     ]}}
  }
 }
 {
  id=Enveloper_5
  children [Enveloper_1]
  properties {
   bLoop=true
   bLoopIndefinitely=true
   loopEnd=40
   pitchCurve {editorCurveData {keys [{time=0}]}}
   volumeCurve {editorCurveData {keys [
      {arriveTangent=-0.01, interpMode=RCIM_Cubic, time=0, value=0.1}
      <at 20 1 RCIM_Cubic>
      {interpMode=RCIM_Cubic, leaveTangent=0.01, time=39.9, value=0.1}
     ]}}
  }
 }
 {
  id=Enveloper_6
  children [Enveloper_1]
  properties {
   bLoop=true
   bLoopIndefinitely=true
   loopEnd=30
   pitchCurve {editorCurveData {keys [{time=0}]}}
   volumeCurve {editorCurveData {keys [
      {arriveTangent=-0.01, interpMode=RCIM_Cubic, time=0, value=0.5}
      <at 15 1 RCIM_Cubic>
      {interpMode=RCIM_Cubic, leaveTangent=0.01, time=29.9, value=0.5}
     ]}}
  }
 }
 {
  id=Enveloper_7
  children [Enveloper_0]
  properties {
   bLoop=true
   bLoopIndefinitely=true
   loopEnd=40
   pitchCurve {editorCurveData {keys [{time=0}]}}
   volumeCurve {editorCurveData {keys [
      {arriveTangent=-0.01, interpMode=RCIM_Cubic, time=0, value=1}
      <at 20 0.5 RCIM_Cubic>
      {interpMode=RCIM_Cubic, leaveTangent=0.01, time=39.9, value=1}
     ]}}
  }
 }
 {
  id=Enveloper_8
  children [Enveloper_0]
  properties {
   bLoop=true
   bLoopIndefinitely=true
   loopEnd=40
   pitchCurve {editorCurveData {keys [{time=0}]}}
   volumeCurve {editorCurveData {keys [
      {arriveTangent=-0.01, interpMode=RCIM_Cubic, time=0, value=0.1}
      <at 20 1 RCIM_Cubic>
      {interpMode=RCIM_Cubic, leaveTangent=0.01, time=39.9, value=0.1}
     ]}}
  }
 }
 {
  id=Enveloper_9
  children [Enveloper_0]
  properties {
   bLoop=true
   bLoopIndefinitely=true
   loopEnd=30
   pitchCurve {editorCurveData {keys [{time=0}]}}
   volumeCurve {editorCurveData {keys [
      {arriveTangent=-0.01, interpMode=RCIM_Cubic, time=0, value=0.5}
      <at 15 1 RCIM_Cubic>
      {interpMode=RCIM_Cubic, leaveTangent=0.01, time=29.9, value=0.5}
     ]}}
  }
 }
 {id=QualityLevel_0, children [Random_3,Random_3,Random_4,Random_4]}
 {id=Random_0, children [WavePlayer_0,WavePlayer_6], properties {weights [1,1]}}
 {id=Random_1, children [WavePlayer_7,WavePlayer_8], properties {weights [1,1]}}
 {id=Random_3, children [Enveloper_2,Enveloper_5,Enveloper_6], properties {weights [1,1,1]}}
 {id=Random_4, children [Enveloper_7,Enveloper_8,Enveloper_9], properties {weights [1,1,1]}}
 {id=WavePlayer_0, properties {bLooping=true, soundWaveAssetPtr=/AmbientAudioContent/Waves/Biomes/Ambience_Japanese_Cicadas_Loud_Loop_01_Quad}}
 {id=WavePlayer_6, properties {bLooping=true, soundWaveAssetPtr=/AmbientAudioContent/Waves/Biomes/Ambience_Japanese_Cicadas_Loud_Loop_02_Quad}}
 {
  id=WavePlayer_7
  properties {bLooping=true, soundWaveAssetPtr=/AmbientAudioContent/Waves/Biomes/Ambience_Japanese_Cicadas_Loud_Loop_01_Stereo}
 }
 {
  id=WavePlayer_8
  properties {bLooping=true, soundWaveAssetPtr=/AmbientAudioContent/Waves/Biomes/Ambience_Japanese_Cicadas_Loud_Loop_02_Stereo}
 }
]
properties {
 attenuationSettings=&Fort_Ambient_Athena_Att
 maxDistance=2097152
 soundClassObject=&Fort_Ambient_Outdoors_Athena
 volumeMultiplier=0.8
 preEffectBusSends [{}]
}
