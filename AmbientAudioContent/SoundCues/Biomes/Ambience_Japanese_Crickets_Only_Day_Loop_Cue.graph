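# Sound cue for a looping daytime Japanese crickets biome ambience, using outdoor ambient class and attenuation.
# A quality switch sends its first two levels to quad cricket loops and its last two to a single stereo loop.
# Each branch randomly picks a looping wave or an empty slot.
# Nested looping envelopers vary pitch over 30 seconds and dip volume over 25 seconds, so their mismatched cycles drift against each other.
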SoundCue /AmbientAudioContent/SoundCues/Biomes/Ambience_Japanese_Crickets_Only_Day_Loop_Cue
root=QualityLevel_0
nodes [
 {
  id=Enveloper_0
  children [Random_0]
  properties {
   bLoop=true
   bLoopIndefinitely=true
   loopEnd=30
   pitchCurve {editorCurveData {keys [{interpMode=RCIM_Cubic, time=0}, <at 15.9468 0.98 RCIM_Cubic>, {interpMode=RCIM_Cubic, time=30}]}}
   volumeCurve {editorCurveData {keys [<at 0 1>]}}
  }
 }
 {
  id=Enveloper_1
  children [Random_1]
  properties {
   bLoop=true
   bLoopIndefinitely=true
   loopEnd=30
   pitchCurve {editorCurveData {keys [{interpMode=RCIM_Cubic, time=0}, <at 15.9468 0.98 RCIM_Cubic>, {interpMode=RCIM_Cubic, time=30}]}}
   volumeCurve {editorCurveData {keys [<at 0 1>]}}
  }
 }
 {
  id=Enveloper_2
  children [Enveloper_0]
  properties {
   bLoop=true
   bLoopIndefinitely=true
   loopEnd=25
   pitchCurve {editorCurveData {keys [{time=0}]}}
   volumeCurve {editorCurveData {keys [
      {arriveTangent=-0.01, interpMode=RCIM_Cubic, time=0, value=1}
      <at 12.5 0.2 RCIM_Cubic>
      {interpMode=RCIM_Cubic, leaveTangent=0.01, time=24.9, value=1}
     ]}}
  }
 }
 {
  id=Enveloper_5
  children [Enveloper_1]
  properties {
   bLoop=true
   bLoopIndefinitely=true
   loopEnd=25
   pitchCurve {editorCurveData {keys [{time=0}]}}
   volumeCurve {editorCurveData {keys [
      {arriveTangent=-0.01, interpMode=RCIM_Cubic, time=0, value=1}
      <at 12.5 0.2 RCIM_Cubic>
      {interpMode=RCIM_Cubic, leaveTangent=0.01, time=24.9, value=1}
     ]}}
  }
 }
 {id=QualityLevel_0, children [Enveloper_2,Enveloper_2,Enveloper_5,Enveloper_5]}
 {id=Random_0, children [WavePlayer_2,WavePlayer_3,WavePlayer_6,null], properties {weights [1.5,1.5,1.5,0.5]}}
 {id=Random_1, children [WavePlayer_7,WavePlayer_7,WavePlayer_7,null], properties {weights [1,1,1,1]}}
 {id=WavePlayer_2, properties {bLooping=true, soundWaveAssetPtr=/AmbientAudioContent/Waves/Biomes/Ambience_Crickets_Only_Loop_05_Quad}}
 {id=WavePlayer_3, properties {bLooping=true, soundWaveAssetPtr=/AmbientAudioContent/Waves/Biomes/Ambience_Crickets_Only_Loop_04_Quad}}
 {id=WavePlayer_6, properties {bLooping=true, soundWaveAssetPtr=/AmbientAudioContent/Waves/Biomes/Ambience_Crickets_Only_Loop_01_Quad}}
 {id=WavePlayer_7, properties {bLooping=true, soundWaveAssetPtr=/AmbientAudioContent/Waves/Biomes/Ambience_Crickets_Only_Loop_04_Stereo}}
]
properties {
 attenuationSettings=/AmbientAudioBase/Attenuations/Fort_Ambient_Athena_Att
 maxDistance=2097152
 soundClassObject=&Fort_Ambient_Outdoors_Athena
 volumeMultiplier=0.15
 preEffectBusSends [{}]
}
