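# Looping daytime ambience cue for the Japanese Grassland biome, using the outdoor ambient sound class and attenuation.
# A quality switch plays a quad bed mixed with stereo winds in its first two slots, and a single stereo loop in the other two.
# Nested looping envelopers cycle pitch and volume over differing periods to vary each loop, with deeper volume dips and randomized volume on the winds.
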SoundCue /AmbientAudioContent/SoundCues/Biomes/Ambience_Japanese_Grassland_Day_Loop_Cue
root=QualityLevel_0
nodes [
 {
  id=Enveloper_0
  children [WavePlayer_2]
  properties {
   bLoop=true
   bLoopIndefinitely=true
   loopEnd=30
   pitchCurve {editorCurveData {keys [
      {arriveTangent=-0.0013, interpMode=RCIM_Cubic, leaveTangent=-0.0013, time=0}
      <at 15 0.98 RCIM_Cubic>
      {arriveTangent=0.0013, interpMode=RCIM_Cubic, leaveTangent=0.0013, time=30}
     ]}}
   volumeCurve {editorCurveData {keys [<at 0 1>]}}
  }
 }
 {
  id=Enveloper_1
  children [WavePlayer_3]
  properties {
   bLoop=true
   bLoopIndefinitely=true
   loopEnd=30
   pitchCurve {editorCurveData {keys [
      {arriveTangent=-0.0013, interpMode=RCIM_Cubic, leaveTangent=-0.0013, time=0}
      <at 15 0.98 RCIM_Cubic>
      {arriveTangent=0.0013, interpMode=RCIM_Cubic, leaveTangent=0.0013, time=30}
     ]}}
   volumeCurve {editorCurveData {keys [<at 0 1>]}}
  }
 }
 {
  id=Enveloper_3
  children [Enveloper_0]
  properties {
   bLoop=true
   bLoopIndefinitely=true
   loopEnd=50
   pitchCurve {editorCurveData {keys [{time=0}]}}
   volumeCurve {editorCurveData {keys [
      {arriveTangent=-0.01, interpMode=RCIM_Cubic, time=0, value=1}
      <at 25 0.75 RCIM_Cubic>
      {interpMode=RCIM_Cubic, leaveTangent=0.01, time=49.9, value=1}
     ]}}
  }
 }
 {
  id=Enveloper_4
  children [Modulator_0]
  properties {
   bLoop=true
   bLoopIndefinitely=true
   loopEnd=20
   pitchCurve {editorCurveData {keys [
      {arriveTangent=-0.0013, interpMode=RCIM_Cubic, leaveTangent=-0.0013, time=0}
      <at 15 0.98 RCIM_Cubic>
      {arriveTangent=0.0013, interpMode=RCIM_Cubic, leaveTangent=0.0013, time=30}
     ]}}
   volumeCurve {editorCurveData {keys [<at 0 1>]}}
  }
 }
 {
  id=Enveloper_5
  children [Enveloper_4]
  properties {
   bLoop=true
   bLoopIndefinitely=true
   loopEnd=25
   pitchCurve {editorCurveData {keys [{time=0}]}}
   volumeCurve {editorCurveData {keys [
      {arriveTangent=-0.01, interpMode=RCIM_Cubic, time=0, value=1}
      <at 12.5 0.4 RCIM_Cubic>
      {interpMode=RCIM_Cubic, leaveTangent=0.01, time=24.5, value=1}
     ]}}
  }
 }
 {
  id=Enveloper_6
  children [Enveloper_1]
  properties {
   bLoop=true
   bLoopIndefinitely=true
   loopEnd=50
   pitchCurve {editorCurveData {keys [{time=0}]}}
   volumeCurve {editorCurveData {keys [
      {arriveTangent=-0.01, interpMode=RCIM_Cubic, time=0, value=1}
      <at 25 0.75 RCIM_Cubic>
      {interpMode=RCIM_Cubic, leaveTangent=0.01, time=49.9, value=1}
     ]}}
  }
 }
 {id=Mixer_0, children [Enveloper_5,Enveloper_3], properties {inputVolume [0.4,0.6]}}
 {id=Modulator_0, children [WavePlayer_1], properties {volumeMin=0.75}}
 {id=QualityLevel_0, children [Mixer_0,Mixer_0,Enveloper_6,Enveloper_6]}
 {
  id=WavePlayer_1
  properties {bLooping=true, soundWaveAssetPtr=/AmbientAudioContent/Waves/Biomes/Ambience_Japanese_Grassland_Day_Winds_Stereo}
 }
 {id=WavePlayer_2, properties {bLooping=true, soundWaveAssetPtr=/AmbientAudioContent/Waves/Biomes/Ambience_Japanese_Grassland_Day_Quad_Loop}}
 {id=WavePlayer_3, properties {bLooping=true, soundWaveAssetPtr=/AmbientAudioContent/Waves/Biomes/Ambience_Japanese_Grassland_Day_Stereo_Loop}}
]
properties {
 attenuationSettings=/AmbientAudioBase/Attenuations/Fort_Ambient_Athena_Att
 maxDistance=2097152
 soundClassObject=&Fort_Ambient_Outdoors_Athena
 volumeMultiplier=0.6
 preEffectBusSends [{}]
}
