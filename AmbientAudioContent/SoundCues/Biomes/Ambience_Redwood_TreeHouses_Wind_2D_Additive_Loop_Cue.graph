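# Looping 2D additive wind ambience cue for the Redwood treehouses biome, on the outdoor ambient sound class and capped at one instance.
# A randomly picked looping window-wind bed, under slow looping envelopes, mixes with randomly delayed, modulated town gust and autumn-leaf one-shots.
# Further looping envelopers slowly vary pitch and volume of the whole mix.
# Only the first two quality-level slots feed this chain; the other two are empty.
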SoundCue /AmbientAudioContent/SoundCues/Biomes/Ambience_Redwood_TreeHouses_Wind_2D_Additive_Loop_Cue
root=QualityLevel_0
nodes [
 {id=Delay_1, children [Modulator_1], properties {delayMax=3, delayMin=1}}
 {id=Delay_3, children [Modulator_3], properties {delayMax=8, delayMin=3}}
 {
  id=Enveloper_0
  children [Mixer_1]
  properties {
   bLoop=true
   bLoopIndefinitely=true
   loopEnd=30
   pitchCurve {editorCurveData {keys [{interpMode=RCIM_Cubic, time=0}, <at 15.9468 0.9181 RCIM_Cubic>, {interpMode=RCIM_Cubic, time=30}]}}
   volumeCurve {editorCurveData {keys [<at 0 1>]}}
  }
 }
 {
  id=Enveloper_3
  children [Enveloper_0]
  properties {
   bLoop=true
   bLoopIndefinitely=true
   loopEnd=20
   pitchCurve {editorCurveData {keys [{time=0}]}}
   volumeCurve {editorCurveData {keys [
      {arriveTangent=-0.01, interpMode=RCIM_Cubic, time=0, value=1}
      <at 10 1.1 RCIM_Cubic>
      {interpMode=RCIM_Cubic, leaveTangent=0.01, time=20, value=1}
     ]}}
  }
 }
 {
  id=Enveloper_4
  children [Random_9]
  properties {
   bLoop=true
   bLoopIndefinitely=true
   loopEnd=30
   pitchCurve {editorCurveData {keys [{interpMode=RCIM_Cubic, time=0}, <at 15 0.9 RCIM_Cubic>, {interpMode=RCIM_Cubic, time=30}]}}
   volumeCurve {editorCurveData {keys [<at 0 1>]}}
  }
 }
 {
  id=Enveloper_5
  children [Enveloper_4]
  properties {
   bLoop=true
   bLoopIndefinitely=true
   loopEnd=60
   pitchCurve {editorCurveData {keys [{time=0}]}}
   volumeCurve {editorCurveData {keys [
      {arriveTangent=-0.01, interpMode=RCIM_Cubic, time=0, value=1}
      <at 10 1.1 RCIM_Cubic>
      {arriveTangent=-0.0047, interpMode=RCIM_Cubic, leaveTangent=-0.0047, time=20, value=1}
      <at 31.1629 1>
      {arriveTangent=-0.0016, interpMode=RCIM_Cubic, leaveTangent=-0.0016, time=37.937, value=0.9}
      <at 46.7125 0.975>
      <at 60 1 RCIM_Cubic>
     ]}}
  }
 }
 {id=Looping_1, children [Delay_1]}
 {id=Looping_3, children [Delay_3]}
 {id=Mixer_1, children [Looping_3,Looping_1,Enveloper_5], properties {inputVolume [1,1,2]}}
 {id=Modulator_0, children [WavePlayer_48], properties {volumeMax=1.25, volumeMin=1.25}}
 {id=Modulator_1, children [Random_7], properties {pitchMax=1.1, volumeMax=1.75, volumeMin=1.5}}
 {id=Modulator_2, children [WavePlayer_47], properties {volumeMax=1.75, volumeMin=1.75}}
 {id=Modulator_3, children [Random_6], properties {pitchMax=1.1, volumeMax=1.75, volumeMin=1.5}}
 {id=Modulator_6, children [WavePlayer_46], properties {volumeMax=1, volumeMin=1}}
 {id=QualityLevel_0, children [Enveloper_3,Enveloper_3,null,null]}
 {
  id=Random_6
  children [WavePlayer_25, WavePlayer_26, WavePlayer_27, WavePlayer_28, WavePlayer_29, WavePlayer_30]
  properties {weights [1,1,1,1,1,1]}
 }
 {id=Random_7, children [WavePlayer_31, WavePlayer_32, WavePlayer_33, WavePlayer_34, WavePlayer_35], properties {weights [1,1,1,1,1]}}
 {id=Random_9, children [Modulator_6,Modulator_2,Modulator_0], properties {weights [1,1,1]}}
 {id=WavePlayer_25, properties {soundWaveAssetPtr=/AmbientAudioContent/Waves/OneShots/Spot_Wind_Autumnal_Leaves_Short_01}}
 {id=WavePlayer_26, properties {soundWaveAssetPtr=/AmbientAudioContent/Waves/OneShots/Spot_Wind_Autumnal_Leaves_Short_02}}
 {id=WavePlayer_27, properties {soundWaveAssetPtr=/AmbientAudioContent/Waves/OneShots/Spot_Wind_Autumnal_Leaves_Short_05}}
 {id=WavePlayer_28, properties {soundWaveAssetPtr=/AmbientAudioContent/Waves/OneShots/Spot_Wind_Autumnal_Leaves_Short_04}}
 {id=WavePlayer_29, properties {soundWaveAssetPtr=/AmbientAudioContent/Waves/OneShots/Spot_Wind_Autumnal_Leaves_Short_03}}
 {id=WavePlayer_30, properties {soundWaveAssetPtr=/AmbientAudioContent/Waves/OneShots/Spot_Wind_Autumnal_Leaves_Short_06}}
 {id=WavePlayer_31, properties {soundWaveAssetPtr=/AmbientAudioContent/Waves/OneShots/Spot_Town_Wind_Gust_01}}
 {id=WavePlayer_32, properties {soundWaveAssetPtr=/AmbientAudioContent/Waves/OneShots/Spot_Town_Wind_Gust_02}}
 {id=WavePlayer_33, properties {soundWaveAssetPtr=/AmbientAudioContent/Waves/OneShots/Spot_Town_Wind_Gust_03}}
 {id=WavePlayer_34, properties {soundWaveAssetPtr=/AmbientAudioContent/Waves/OneShots/Spot_Town_Wind_Gust_04}}
 {id=WavePlayer_35, properties {soundWaveAssetPtr=/AmbientAudioContent/Waves/OneShots/Spot_Town_Wind_Gust_05}}
 {id=WavePlayer_46, properties {bLooping=true, soundWaveAssetPtr=/AmbientAudioContent/Waves/SpottedLoops/Spot_Olympus_Window_Wind_01}}
 {id=WavePlayer_47, properties {bLooping=true, soundWaveAssetPtr=/AmbientAudioContent/Waves/SpottedLoops/Spot_Olympus_Window_Wind_02}}
 {id=WavePlayer_48, properties {bLooping=true, soundWaveAssetPtr=/AmbientAudioContent/Waves/SpottedLoops/Spot_Olympus_Window_Wind_03}}
]
properties {
 attenuationSettings=&Fort_Ambient_Athena_Att
 bEnableBusSends=false
 bOverrideConcurrency=true
 maxDistance=2097152
 soundClassObject=&Fort_Ambient_Outdoors_Athena
 volumeMultiplier=0.25
 attenuationOverrides {
  attenuationShape=Box
  bAttenuate=false
  bSpatialize=false
  falloffDistance=546.9495
  manualReverbSendLevel=0.6
  nonSpatializedRadiusEnd=600
  nonSpatializedRadiusStart=1600
  reverbSendMethod=Manual
  stereoSpread=2000
  attenuationShapeExtents=<600 1119.4096 600>
 }
 concurrencyOverrides {maxCount=1, resolutionRule=PreventNew}
 preEffectBusSends [{}]
}
