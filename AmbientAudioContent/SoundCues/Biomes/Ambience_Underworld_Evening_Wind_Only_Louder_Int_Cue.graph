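# Looping indoor ambience cue for Underworld evening wind, going by its name a louder interior variant, on the indoor ambient sound class.
# A quality-level node sends its first two tiers to a quad wave and its last two to a stereo wave.
# Each wave runs through two nested looping envelopers, a 30-second pitch curve and a 20-second volume dip, whose unequal periods vary the loop.
# Ambient attenuation and two-max, prevent-new concurrency apply.
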SoundCue /AmbientAudioContent/SoundCues/Biomes/Ambience_Underworld_Evening_Wind_Only_Louder_Int_Cue
root=QualityLevel_0
nodes [
 {
  id=Enveloper_0
  children [WavePlayer_2]
  properties {
   bLoop=true
   bLoopIndefinitely=true
   loopEnd=30
   pitchCurve {editorCurveData {keys [
      {arriveTangent=-0.0033, interpMode=RCIM_Cubic, leaveTangent=-0.0033, time=0}
      <at 15 0.95 RCIM_Cubic>
      {arriveTangent=0.0033, interpMode=RCIM_Cubic, leaveTangent=0.0033, time=30}
     ]}}
   volumeCurve {editorCurveData {keys [<at 0 1>]}}
  }
 }
 {
  id=Enveloper_1
  children [WavePlayer_3]
  properties {
   bLoop=true
   bLoopIndefinitely=true
   loopEnd=30
   pitchCurve {editorCurveData {keys [
      {arriveTangent=-0.0033, interpMode=RCIM_Cubic, leaveTangent=-0.0033, time=0}
      <at 15 0.95 RCIM_Cubic>
      {arriveTangent=0.0033, interpMode=RCIM_Cubic, leaveTangent=0.0033, time=30}
     ]}}
   volumeCurve {editorCurveData {keys [<at 0 1>]}}
  }
 }
 {
  id=Enveloper_2
  children [Enveloper_1]
  properties {
   bLoop=true
   bLoopIndefinitely=true
   loopEnd=20
   pitchCurve {editorCurveData {keys [{time=0}]}}
   volumeCurve {editorCurveData {keys [
      {arriveTangent=-0.01, interpMode=RCIM_Cubic, leaveTangent=-0.01, time=0, value=1}
      <at 10 0.9 RCIM_Cubic>
      {arriveTangent=0.01, interpMode=RCIM_Cubic, leaveTangent=0.01, time=20, value=1}
     ]}}
  }
 }
 {
  id=Enveloper_3
  children [Enveloper_0]
  properties {
   bLoop=true
   bLoopIndefinitely=true
   loopEnd=20
   pitchCurve {editorCurveData {keys [{time=0}]}}
   volumeCurve {editorCurveData {keys [
      {arriveTangent=-0.01, interpMode=RCIM_Cubic, leaveTangent=-0.01, time=0, value=1}
      <at 10 0.9 RCIM_Cubic>
      {arriveTangent=0.01, interpMode=RCIM_Cubic, leaveTangent=0.01, time=20, value=1}
     ]}}
  }
 }
 {id=QualityLevel_0, children [Enveloper_3,Enveloper_3,Enveloper_2,Enveloper_2]}
 {
  id=WavePlayer_2
  properties {bLooping=true, soundWaveAssetPtr=/AmbientAudioContent/Waves/Biomes/Ambience_Underworld_Evening_Wind_Only_Loop_Quad}
 }
 {
  id=WavePlayer_3
  properties {bLooping=true, soundWaveAssetPtr=/AmbientAudioContent/Waves/Biomes/Ambience_Underworld_Evening_Wind_Only_Loop_Stereo}
 }
]
properties {
 attenuationSettings=/AmbientAudioBase/Attenuations/Fort_Ambient_Att
 bEnableBusSends=false
 maxDistance=2097152
 soundClassObject=&Fort_Ambient_Indoors_Athena
 volumeMultiplier=0.7
 concurrencySet [/AmbientAudioBase/Concurrency/AmbientConc_2Max_Loop_PreventNew_Default_Concurrency]
}
