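# Looping ambience Sound Cue for the underside of a wooden pontoon dock on water, in the water ambient sound class.
# It mixes a looping deck-water bed with random wood impacts that repeat after random delays with slight pitch variation.
# Two reused wooden-bridge corridor loops, quad and pitched-down stereo, pass through nested looping envelopers whose differing pitch and volume periods add slow drift.
# The mix feeds a shared close-river attenuation node.
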SoundCue /AmbientAudioContent/SoundCues/Biomes/Ambience_Water_Wood_Pontoon_Dock_Underside_Loop_Cue
root=Attenuation_1
nodes [
 {
  id=Attenuation_1
  children [Mixer_0]
  properties {attenuationSettings=/Game/Athena/Sounds/Water/Attenuations/Apollo_Water_River_Close_Focused_Att}
 }
 {id=Delay_0, children [Modulator_0], properties {delayMax=6, delayMin=3}}
 {
  id=Enveloper_1
  children [WavePlayer_1]
  properties {
   bLoop=true
   bLoopIndefinitely=true
   loopEnd=30
   pitchCurve {editorCurveData {keys [{interpMode=RCIM_Cubic, time=0}, <at 15.9468 0.9181 RCIM_Cubic>, {interpMode=RCIM_Cubic, time=30}]}}
   volumeCurve {editorCurveData {keys [<at 0 1>]}}
  }
 }
 {
  id=Enveloper_2
  children [Enveloper_1]
  properties {
   bLoop=true
   bLoopIndefinitely=true
   loopEnd=20
   pitchCurve {editorCurveData {keys [{time=0}]}}
   volumeCurve {editorCurveData {keys [
      {arriveTangent=0, interpMode=RCIM_Cubic, leaveTangent=0, time=0, value=1}
      {arriveTangent=0, interpMode=RCIM_Cubic, leaveTangent=0, time=10, value=0.9}
      {arriveTangent=0, interpMode=RCIM_Cubic, leaveTangent=0, time=20, value=0.9992}
     ]}}
  }
 }
 {
  id=Enveloper_4
  children [Modulator_2]
  properties {
   bLoop=true
   bLoopIndefinitely=true
   loopEnd=30
   pitchCurve {editorCurveData {keys [{interpMode=RCIM_Cubic, time=0}, <at 15.9468 0.9181 RCIM_Cubic>, {interpMode=RCIM_Cubic, time=30}]}}
   volumeCurve {editorCurveData {keys [<at 0 1>]}}
  }
 }
 {
  id=Enveloper_5
  children [Enveloper_4]
  properties {
   bLoop=true
   bLoopIndefinitely=true
   loopEnd=20
   pitchCurve {editorCurveData {keys [{time=0}]}}
   volumeCurve {editorCurveData {keys [
      {arriveTangent=0, interpMode=RCIM_Cubic, leaveTangent=0, time=0, value=1}
      {arriveTangent=0, interpMode=RCIM_Cubic, leaveTangent=0, time=10, value=0.9}
      {arriveTangent=0, interpMode=RCIM_Cubic, leaveTangent=0, time=20, value=0.9992}
     ]}}
  }
 }
 {id=Looping_0, children [Delay_0]}
 {id=Mixer_0, children [Looping_0,WavePlayer_2,Mixer_1,Mixer_2], properties {inputVolume [0.6,0.6,1,1]}}
 {id=Mixer_1, children [Enveloper_2], properties {inputVolume [1.5]}}
 {id=Mixer_2, children [Enveloper_5], properties {inputVolume [1.5]}}
 {id=Modulator_0, children [Random_0], properties {pitchMax=1.1, pitchMin=0.9}}
 {id=Modulator_2, children [WavePlayer_0], properties {pitchMax=0.85, pitchMin=0.85, volumeMax=1, volumeMin=1}}
 {id=Random_0, children [WavePlayer_3,WavePlayer_4,WavePlayer_5,WavePlayer_6,WavePlayer_7,WavePlayer_8], properties {weights [1,1,1,1,1,1]}}
 {
  id=WavePlayer_0
  properties {bLooping=true, soundWaveAssetPtr=/Game/Sounds/Artemis/POI/Normal_Bridges/Bridges_Wooden_Corridor/Ambience_Bridge_Wooden_Corridor_Stereo}
 }
 {
  id=WavePlayer_1
  properties {bLooping=true, soundWaveAssetPtr=/Game/Sounds/Artemis/POI/Normal_Bridges/Bridges_Wooden_Corridor/Ambience_Bridge_Wooden_Corridor_Quad}
 }
 {
  id=WavePlayer_2
  properties {bLooping=true, soundWaveAssetPtr=/Game/Sounds/Asteria/POI/RainForest_Village/Spots/Spot_Rainforest_Village_Pontoon_Deck_Water_01}
 }
 {id=WavePlayer_3, properties {soundWaveAssetPtr=/Game/Athena/Sounds/Water/Ambient/Spot_Water_Pontoon_Wood_Impact_01}}
 {id=WavePlayer_4, properties {soundWaveAssetPtr=/Game/Athena/Sounds/Water/Ambient/Spot_Water_Pontoon_Wood_Impact_02}}
 {id=WavePlayer_5, properties {soundWaveAssetPtr=/Game/Athena/Sounds/Water/Ambient/Spot_Water_Pontoon_Wood_Impact_03}}
 {id=WavePlayer_6, properties {soundWaveAssetPtr=/Game/Athena/Sounds/Water/Ambient/Spot_Water_Pontoon_Wood_Impact_04}}
 {id=WavePlayer_7, properties {soundWaveAssetPtr=/Game/Athena/Sounds/Water/Ambient/Spot_Water_Pontoon_Wood_Impact_05}}
 {id=WavePlayer_8, properties {soundWaveAssetPtr=/Game/Athena/Sounds/Water/Ambient/Spot_Water_Pontoon_Wood_Impact_06}}
]
properties {
 maxDistance=4512
 soundClassObject=/Game/Sounds/FortSoundClassesAndModes/Fort_Water_Ambient
 volumeMultiplier=1
 attenuationOverrides {
  bAttenuateWithLPF=true
  falloffDistance=6000
  lPFFrequencyAtMax=500
  nonSpatializedRadiusStart=700
  stereoSpread=1000
  attenuationShapeExtents {x=925}
 }
 concurrencyOverrides {maxCount=1, resolutionRule=PreventNew}
 concurrencySet [null]
}
