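# Sound cue for quiet one-shot daytime bird calls in the Knights forest biome, using bird attenuation and the outdoor ambient one-shot sound class.
# A quality-level root sends the first two levels to a random pick between an empty branch and a large call pool.
# The last two levels pick between two small pools, both excluded from branch culling.
# Modulators above each branch randomize pitch and volume for variety.
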SoundCue /AmbientAudioContent/SoundCues/Biomes/Spot_Knights_Forest_Day_Bird_Cue
root=QualityLevel_0
nodes [
 {id=Modulator_2, children [null], properties {pitchMin=0.3, volumeMax=0.6, volumeMin=0.3}}
 {id=Modulator_3, children [Random_1], properties {pitchMin=0.3, volumeMax=0.6, volumeMin=0.3}}
 {id=Modulator_4, children [Random_8], properties {pitchMin=0.3, volumeMax=0.6, volumeMin=0.3}}
 {id=Modulator_5, children [Random_9], properties {pitchMin=0.3, volumeMax=0.6, volumeMin=0.3}}
 {id=QualityLevel_0, children [Random_7,Random_7,Random_10,Random_10]}
 {
  id=Random_1
  children [
   WavePlayer_62
   WavePlayer_63
   WavePlayer_64
   WavePlayer_65
   WavePlayer_66
   WavePlayer_67
   WavePlayer_68
   WavePlayer_69
   WavePlayer_70
   WavePlayer_71
   WavePlayer_72
   WavePlayer_73
   WavePlayer_106
   WavePlayer_107
   WavePlayer_108
   WavePlayer_109
   WavePlayer_110
   WavePlayer_111
   WavePlayer_112
   WavePlayer_113
   WavePlayer_114
   WavePlayer_115
   WavePlayer_116
   WavePlayer_117
   WavePlayer_118
   WavePlayer_119
   WavePlayer_120
   WavePlayer_121
   WavePlayer_122
   WavePlayer_123
   WavePlayer_124
   WavePlayer_125
  ]
  properties {weights [1,1,1,1,1,1,1,1,1,1,1,1,1,1,1,1,1,1,1,1,1,1,1,1,1,1,1,1,1,1,1,1]}
 }
 {id=Random_10, children [Modulator_4,Modulator_5], properties {bShouldExcludeFromBranchCulling=true, weights [1,1]}}
 {id=Random_7, children [Modulator_2,Modulator_3], properties {weights [1,1]}}
 {
  id=Random_8
  children [WavePlayer_205, WavePlayer_206, WavePlayer_207, WavePlayer_208]
  properties {bShouldExcludeFromBranchCulling=true, weights [1,1,1,1]}
 }
 {id=Random_9, children [WavePlayer_209, WavePlayer_210, WavePlayer_211], properties {bShouldExcludeFromBranchCulling=true, weights [1,1,1]}}
 {id=WavePlayer_106, properties {soundWaveAssetPtr=/AmbientAudioContent/Waves/OneShots/Spot_Knights_Forest_Day_Bird_Call_39}}
 {id=WavePlayer_107, properties {soundWaveAssetPtr=/AmbientAudioContent/Waves/OneShots/Spot_Knights_Forest_Day_Bird_Call_40}}
 {id=WavePlayer_108, properties {soundWaveAssetPtr=/AmbientAudioContent/Waves/OneShots/Spot_Knights_Forest_Day_Bird_Call_41}}
 {id=WavePlayer_109, properties {soundWaveAssetPtr=/AmbientAudioContent/Waves/OneShots/Spot_Knights_Forest_Day_Bird_Call_42}}
 {id=WavePlayer_110, properties {soundWaveAssetPtr=/AmbientAudioContent/Waves/OneShots/Spot_Knights_Forest_Day_Bird_Call_43}}
 {id=WavePlayer_111, properties {soundWaveAssetPtr=/AmbientAudioContent/Waves/OneShots/Spot_Knights_Forest_Day_Bird_Call_44}}
 {id=WavePlayer_112, properties {soundWaveAssetPtr=/AmbientAudioContent/Waves/OneShots/Spot_Knights_Forest_Day_Bird_Call_45}}
 {id=WavePlayer_113, properties {soundWaveAssetPtr=/AmbientAudioContent/Waves/OneShots/Spot_Knights_Forest_Day_Bird_Call_46}}
 {id=WavePlayer_114, properties {soundWaveAssetPtr=/AmbientAudioContent/Waves/OneShots/Spot_Knights_Forest_Day_Bird_Call_47}}
 {id=WavePlayer_115, properties {soundWaveAssetPtr=/AmbientAudioContent/Waves/OneShots/Spot_Knights_Forest_Day_Bird_Call_48}}
 {id=WavePlayer_116, properties {soundWaveAssetPtr=/AmbientAudioContent/Waves/OneShots/Spot_Knights_Forest_Day_Bird_Call_49}}
 {id=WavePlayer_117, properties {soundWaveAssetPtr=/AmbientAudioContent/Waves/OneShots/Spot_Knights_Forest_Day_Bird_Call_50}}
 {id=WavePlayer_118, properties {soundWaveAssetPtr=/AmbientAudioContent/Waves/OneShots/Spot_Knights_Forest_Day_Bird_Call_51}}
 {id=WavePlayer_119, properties {soundWaveAssetPtr=/AmbientAudioContent/Waves/OneShots/Spot_Knights_Forest_Day_Bird_Call_52}}
 {id=WavePlayer_120, properties {soundWaveAssetPtr=/AmbientAudioContent/Waves/OneShots/Spot_Knights_Forest_Day_Bird_Call_53}}
 {id=WavePlayer_121, properties {soundWaveAssetPtr=/AmbientAudioContent/Waves/OneShots/Spot_Knights_Forest_Day_Bird_Call_54}}
 {id=WavePlayer_122, properties {soundWaveAssetPtr=/AmbientAudioContent/Waves/OneShots/Spot_Knights_Forest_Day_Bird_Call_55}}
 {id=WavePlayer_123, properties {soundWaveAssetPtr=/AmbientAudioContent/Waves/OneShots/Spot_Knights_Forest_Day_Bird_Call_56}}
 {id=WavePlayer_124, properties {soundWaveAssetPtr=/AmbientAudioContent/Waves/OneShots/Spot_Knights_Forest_Day_Bird_Call_58}}
 {id=WavePlayer_125, properties {soundWaveAssetPtr=/AmbientAudioContent/Waves/OneShots/Spot_Knights_Forest_Day_Bird_Call_59}}
 {id=WavePlayer_205, properties {soundWaveAssetPtr=/AmbientAudioContent/Waves/OneShots/Spot_Knights_Forest_Day_Bird_Call_16}}
 {id=WavePlayer_206, properties {soundWaveAssetPtr=/AmbientAudioContent/Waves/OneShots/Spot_Knights_Forest_Day_Bird_Call_12}}
 {id=WavePlayer_207, properties {soundWaveAssetPtr=/AmbientAudioContent/Waves/OneShots/Spot_Knights_Forest_Day_Bird_Call_11}}
 {id=WavePlayer_208, properties {soundWaveAssetPtr=/AmbientAudioContent/Waves/OneShots/Spot_Knights_Forest_Day_Bird_Call_01}}
 {id=WavePlayer_209, properties {soundWaveAssetPtr=&Spot_Knights_Forest_Day_Bird_Call_39}}
 {id=WavePlayer_210, properties {soundWaveAssetPtr=/AmbientAudioContent/Waves/OneShots/Spot_Knights_Forest_Day_Bird_Call_38}}
 {id=WavePlayer_211, properties {soundWaveAssetPtr=/AmbientAudioContent/Waves/OneShots/Spot_Knights_Forest_Day_Bird_Call_36}}
 {id=WavePlayer_62, properties {soundWaveAssetPtr=/AmbientAudioContent/Waves/OneShots/Spot_Knights_Forest_Day_Bird_Call_26}}
 {id=WavePlayer_63, properties {soundWaveAssetPtr=/AmbientAudioContent/Waves/OneShots/Spot_Knights_Forest_Day_Bird_Call_27}}
 {id=WavePlayer_64, properties {soundWaveAssetPtr=/AmbientAudioContent/Waves/OneShots/Spot_Knights_Forest_Day_Bird_Call_28}}
 {id=WavePlayer_65, properties {soundWaveAssetPtr=/AmbientAudioContent/Waves/OneShots/Spot_Knights_Forest_Day_Bird_Call_29}}
 {id=WavePlayer_66, properties {soundWaveAssetPtr=/AmbientAudioContent/Waves/OneShots/Spot_Knights_Forest_Day_Bird_Call_30}}
 {id=WavePlayer_67, properties {soundWaveAssetPtr=/AmbientAudioContent/Waves/OneShots/Spot_Knights_Forest_Day_Bird_Call_31}}
 {id=WavePlayer_68, properties {soundWaveAssetPtr=/AmbientAudioContent/Waves/OneShots/Spot_Knights_Forest_Day_Bird_Call_33}}
 {id=WavePlayer_69, properties {soundWaveAssetPtr=/AmbientAudioContent/Waves/OneShots/Spot_Knights_Forest_Day_Bird_Call_34}}
 {id=WavePlayer_70, properties {soundWaveAssetPtr=/AmbientAudioContent/Waves/OneShots/Spot_Knights_Forest_Day_Bird_Call_35}}
 {id=WavePlayer_71, properties {soundWaveAssetPtr=&Spot_Knights_Forest_Day_Bird_Call_36}}
 {id=WavePlayer_72, properties {soundWaveAssetPtr=/AmbientAudioContent/Waves/OneShots/Spot_Knights_Forest_Day_Bird_Call_37}}
 {id=WavePlayer_73, properties {soundWaveAssetPtr=&Spot_Knights_Forest_Day_Bird_Call_38}}
]
properties {
 attenuationSettings=&Fort_TileAmbient_OneShot_Bird_Att
 bEnableBusSends=false
 maxDistance=8005
 soundClassObject=&Fort_Ambient_OneShots_Outdoors
 volumeMultiplier=0.05
 preEffectBusSends [{}]
}
