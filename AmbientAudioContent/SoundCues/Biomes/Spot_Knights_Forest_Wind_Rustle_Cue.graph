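# Sound cue for a spot one-shot wind rustle in the Knights Forest biome, reusing Grassland rustle waves.
# A Random node picks among nine equally weighted waves and is excluded from branch culling, feeding a Modulator that varies volume and slightly raises pitch.
# All four quality-level inputs share that chain.
# Output is quiet, uses wind one-shot tile-ambient attenuation and plays in the ambient one-shots sound class.
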SoundCue /AmbientAudioContent/SoundCues/Biomes/Spot_Knights_Forest_Wind_Rustle_Cue
root=QualityLevel_0
nodes [
 {id=Modulator_1, children [Random_2], properties {pitchMax=1.1, pitchMin=1, volumeMax=1.1, volumeMin=0.9}}
 {id=QualityLevel_0, children [Modulator_1,Modulator_1,Modulator_1,Modulator_1]}
 {
  id=Random_2
  children [WavePlayer_8, WavePlayer_9, WavePlayer_10, WavePlayer_11, WavePlayer_12, WavePlayer_20, WavePlayer_22, WavePlayer_23, WavePlayer_24]
  properties {bShouldExcludeFromBranchCulling=true, weights [1,1,1,1,1,1,1,1,1]}
 }
 {id=WavePlayer_10, properties {soundWaveAssetPtr=/AmbientAudioContent/Waves/OneShots/Spot_Grassland_Wind_Rustle_08}}
 {id=WavePlayer_11, properties {soundWaveAssetPtr=/AmbientAudioContent/Waves/OneShots/Spot_Grassland_Wind_Rustle_07}}
 {id=WavePlayer_12, properties {soundWaveAssetPtr=/AmbientAudioContent/Waves/OneShots/Spot_Grassland_Wind_Rustle_06}}
 {id=WavePlayer_20, properties {soundWaveAssetPtr=/AmbientAudioContent/Waves/OneShots/Spot_Grassland_Wind_Rustle_05}}
 {id=WavePlayer_22, properties {soundWaveAssetPtr=/AmbientAudioContent/Waves/OneShots/Spot_Grassland_Wind_Rustle_04}}
 {id=WavePlayer_23, properties {soundWaveAssetPtr=/AmbientAudioContent/Waves/OneShots/Spot_Grassland_Wind_Rustle_03}}
 {id=WavePlayer_24, properties {soundWaveAssetPtr=/AmbientAudioContent/Waves/OneShots/Spot_Grassland_Wind_Rustle_02}}
 {id=WavePlayer_8, properties {soundWaveAssetPtr=/AmbientAudioContent/Waves/OneShots/Spot_Grassland_Wind_Rustle_10}}
 {id=WavePlayer_9, properties {soundWaveAssetPtr=/AmbientAudioContent/Waves/OneShots/Spot_Grassland_Wind_Rustle_09}}
]
properties {
 attenuationSettings=/Game/Sounds/Attenuation/Fort_TileAmbient_OneShot_Wind_Att
 bEnableBusSends=false
 maxDistance=8000
 soundClassObject=&Fort_Ambient_OneShots
 volumeMultiplier=0.2
}
