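# Sound Cue for a quiet outdoor ambient one-shot of wind rustling trees in the Knights forest biome.
# A Random node picks among ten equally weighted tree rustle waves, and a Modulator varies each play with a slight upward pitch shift and a wide volume range.
# All quality levels share that same branch.
# It uses a wider tile-ambient one-shot attenuation and the outdoor ambient one-shots sound class.
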SoundCue /AmbientAudioContent/SoundCues/Biomes/Spot_Knights_Forest_Wind_Tree_Rustle_02_Cue
root=QualityLevel_0
nodes [
 {id=Modulator_1, children [Random_2], properties {pitchMax=1.1, pitchMin=1, volumeMax=1, volumeMin=0.25}}
 {id=QualityLevel_0, children [Modulator_1,Modulator_1,Modulator_1,Modulator_1]}
 {
  id=Random_2
  children [
   WavePlayer_14
   WavePlayer_15
   WavePlayer_16
   WavePlayer_17
   WavePlayer_18
   WavePlayer_19
   WavePlayer_20
   WavePlayer_21
   WavePlayer_22
   WavePlayer_23
  ]
  properties {weights [1,1,1,1,1,1,1,1,1,1]}
 }
 {id=WavePlayer_14, properties {soundWaveAssetPtr=/AmbientAudioContent/Waves/OneShots/Spot_Knights_Forest_Tree_Rustle_11}}
 {id=WavePlayer_15, properties {soundWaveAssetPtr=/AmbientAudioContent/Waves/OneShots/Spot_Knights_Forest_Tree_Rustle_12}}
 {id=WavePlayer_16, properties {soundWaveAssetPtr=/AmbientAudioContent/Waves/OneShots/Spot_Knights_Forest_Tree_Rustle_13}}
 {id=WavePlayer_17, properties {soundWaveAssetPtr=/AmbientAudioContent/Waves/OneShots/Spot_Knights_Forest_Tree_Rustle_14}}
 {id=WavePlayer_18, properties {soundWaveAssetPtr=/AmbientAudioContent/Waves/OneShots/Spot_Knights_Forest_Tree_Rustle_15}}
 {id=WavePlayer_19, properties {soundWaveAssetPtr=/AmbientAudioContent/Waves/OneShots/Spot_Knights_Forest_Tree_Rustle_16}}
 {id=WavePlayer_20, properties {soundWaveAssetPtr=/AmbientAudioContent/Waves/OneShots/Spot_Knights_Forest_Tree_Rustle_17}}
 {id=WavePlayer_21, properties {soundWaveAssetPtr=/AmbientAudioContent/Waves/OneShots/Spot_Knights_Forest_Tree_Rustle_18}}
 {id=WavePlayer_22, properties {soundWaveAssetPtr=/AmbientAudioContent/Waves/OneShots/Spot_Knights_Forest_Tree_Rustle_19}}
 {id=WavePlayer_23, properties {soundWaveAssetPtr=/AmbientAudioContent/Waves/OneShots/Spot_Knights_Forest_Tree_Rustle_20}}
]
properties {
 attenuationSettings=/Game/Sounds/Attenuation/Fort_TileAmbient_OneShot_Wider_Att
 bEnableBusSends=false
 maxDistance=8000
 soundClassObject=&Fort_Ambient_OneShots_Outdoors
 volumeMultiplier=0.2
 preEffectBusSends [{}]
}
