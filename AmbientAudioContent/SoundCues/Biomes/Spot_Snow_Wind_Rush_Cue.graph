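# Sound cue for one-shot snow wind rush spot ambience in biomes, classed as outdoor weather sound with wind one-shot attenuation.
# A quality-level switch sends its first two slots to a random pick of twelve waves and its last two to a smaller pool of four reused from that set.
# Each pool feeds through a modulator that slightly varies pitch and volume, and the cue plays at reduced volume.
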SoundCue /AmbientAudioContent/SoundCues/Biomes/Spot_Snow_Wind_Rush_Cue
root=QualityLevel_1
nodes [
 {id=Modulator_3, children [Random_0], properties {pitchMax=1.1, pitchMin=0.9, volumeMax=1.1, volumeMin=0.9}}
 {id=Modulator_5, children [Random_7], properties {pitchMax=1.1, pitchMin=0.9, volumeMax=1.1, volumeMin=0.9}}
 {id=QualityLevel_1, children [Modulator_3,Modulator_3,Modulator_5,Modulator_5]}
 {
  id=Random_0
  children [
   WavePlayer_0
   WavePlayer_1
   WavePlayer_2
   WavePlayer_3
   WavePlayer_4
   WavePlayer_5
   WavePlayer_6
   WavePlayer_7
   WavePlayer_8
   WavePlayer_9
   WavePlayer_10
   WavePlayer_11
  ]
  properties {weights [1,1,1,1,1,1,1,1,1,1,1,1]}
 }
 {id=Random_7, children [WavePlayer_38, WavePlayer_39, WavePlayer_40, WavePlayer_41], properties {weights [1,1,1,1]}}
 {id=WavePlayer_0, properties {soundWaveAssetPtr=/AmbientAudioContent/Waves/OneShots/Spot_Snow_Wind_Rush_01}}
 {id=WavePlayer_1, properties {soundWaveAssetPtr=/AmbientAudioContent/Waves/OneShots/Spot_Snow_Wind_Rush_02}}
 {id=WavePlayer_10, properties {soundWaveAssetPtr=/AmbientAudioContent/Waves/OneShots/Spot_Snow_Wind_Rush_11}}
 {id=WavePlayer_11, properties {soundWaveAssetPtr=/AmbientAudioContent/Waves/OneShots/Spot_Snow_Wind_Rush_12}}
 {id=WavePlayer_2, properties {soundWaveAssetPtr=/AmbientAudioContent/Waves/OneShots/Spot_Snow_Wind_Rush_03}}
 {id=WavePlayer_3, properties {soundWaveAssetPtr=/AmbientAudioContent/Waves/OneShots/Spot_Snow_Wind_Rush_04}}
 {id=WavePlayer_38, properties {soundWaveAssetPtr=/AmbientAudioContent/Waves/OneShots/Spot_Snow_Wind_Rush_07}}
 {id=WavePlayer_39, properties {soundWaveAssetPtr=&Spot_Snow_Wind_Rush_02}}
 {id=WavePlayer_4, properties {soundWaveAssetPtr=/AmbientAudioContent/Waves/OneShots/Spot_Snow_Wind_Rush_05}}
 {id=WavePlayer_40, properties {soundWaveAssetPtr=&Spot_Snow_Wind_Rush_01}}
 {id=WavePlayer_41, properties {soundWaveAssetPtr=/AmbientAudioContent/Waves/OneShots/Spot_Snow_Wind_Rush_06}}
 {id=WavePlayer_5, properties {soundWaveAssetPtr=&Spot_Snow_Wind_Rush_06}}
 {id=WavePlayer_6, properties {soundWaveAssetPtr=&Spot_Snow_Wind_Rush_07}}
 {id=WavePlayer_7, properties {soundWaveAssetPtr=/AmbientAudioContent/Waves/OneShots/Spot_Snow_Wind_Rush_08}}
 {id=WavePlayer_8, properties {soundWaveAssetPtr=/AmbientAudioContent/Waves/OneShots/Spot_Snow_Wind_Rush_09}}
 {id=WavePlayer_9, properties {soundWaveAssetPtr=/AmbientAudioContent/Waves/OneShots/Spot_Snow_Wind_Rush_10}}
]
properties {
 attenuationSettings=/Game/Sounds/Attenuation/Fort_TileAmbient_OneShot_Wind_Att
 bEnableBusSends=false
 maxDistance=8000
 soundClassObject=/Game/Sounds/FortSoundClassesAndModes/Fort_Ambient_Outdoors_Weather_Athena
 volumeMultiplier=0.251
 preEffectBusSends [{}]
}
